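# Physics asset for the balloon skeletal mesh of the Supply Drop Donut gadget, giving it a simplified collision shape.
# A single body on the pelvis bone holds one large capsule, offset from the bone and pitched to a steep angle.
# It has no constraints, so there are no jointed bodies.
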PhysicsAsset /Game/Gadgets/Assets/Supply_Drop_Donut/Meshes/Supply_Drop_Donut_Balloon_Physics
bodies {pelvis [{shapeType=Capsule, length=309, radius=114.5, center=<-92.6339 -0.0006 84.6145>, rotation=<rot 78.8008 0 0>}]}
constraints []
