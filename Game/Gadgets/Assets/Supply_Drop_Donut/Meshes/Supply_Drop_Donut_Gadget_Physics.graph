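# Physics asset for the Supply Drop Donut gadget's skeletal mesh, a two-legged, armless rig.
# Each of its nine bones has one capsule body, with a tiny one at the pelvis root and a very large, offset one on the head.
# Constraints chain each leg and the spine-to-head branch back to the pelvis.
# Narrow swing limits and almost no twist allow only slight movement between bodies.
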PhysicsAsset /Game/Gadgets/Assets/Supply_Drop_Donut/Meshes/Supply_Drop_Donut_Gadget_Physics
bodies {
 calf_l [{shapeType=Capsule, length=50.75, radius=33, center=<5 0 0>, rotation=<rot 90 0 0>}]
 calf_r [{shapeType=Capsule, length=50.75, radius=33, center=<5 0 0>, rotation=<rot 90 0 0>}]
 foot_l [{shapeType=Capsule, length=42.25, radius=22.25, center=<9 0 0>, rotation=<rot 90 0 0>}]
 foot_r [{shapeType=Capsule, length=42.25, radius=22.25, center=<9 0 0>, rotation=<rot 90 0 0>}]
 head [{shapeType=Capsule, length=30.5, radius=144.25, center=<105.3422 -79.529 1.241>, rotation=<rot 90 0 0>}]
 pelvis [{shapeType=Capsule, length=1, radius=1}]
 spine [{shapeType=Capsule, length=34.75, radius=65.25, rotation=<rot 90 0 0>}]
 thigh_l [{shapeType=Capsule, length=50.75, radius=44, center=<5 0 0>, rotation=<rot 90 0 0>}]
 thigh_r [{shapeType=Capsule, length=50.75, radius=44, center=<5 0 0>, rotation=<rot 90 0 0>}]
}
constraints [
 {bone1Name=thigh_l, bone2Name=pelvis, swing1LimitDegrees=20, swing2LimitDegrees=15, twistLimitDegrees=2}
 {bone1Name=calf_l, bone2Name=thigh_l, swing1LimitDegrees=20, swing2LimitDegrees=15, twistLimitDegrees=2}
 {bone1Name=foot_l, bone2Name=calf_l, swing1LimitDegrees=30, swing2LimitDegrees=10, twistLimitDegrees=2}
 {bone1Name=thigh_r, bone2Name=pelvis, swing1LimitDegrees=20, swing2LimitDegrees=15, twistLimitDegrees=2}
 {bone1Name=calf_r, bone2Name=thigh_r, swing1LimitDegrees=20, swing2LimitDegrees=15, twistLimitDegrees=2}
 {bone1Name=foot_r, bone2Name=calf_r, swing1LimitDegrees=30, swing2LimitDegrees=10, twistLimitDegrees=2}
 {bone1Name=spine, bone2Name=pelvis, swing1LimitDegrees=10, swing2LimitDegrees=10, twistLimitDegrees=2}
 {bone1Name=head, bone2Name=spine, swing1LimitDegrees=15, swing2LimitDegrees=15, twistLimitDegrees=2}
]
