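# Animation Blueprint for the Supply Drop Donut gadget that procedurally sways its skeleton and adds physics motion, calming it on landing.
# Each update builds noise-wave rotators from per-bone frequencies and a randomly seeded offset.
# These drive additive Modify Bone nodes on offset, pelvis and limbs, with a RigidBody node after the pelvis using overridden low gravity and external force.
# OnLanded randomizes lower noise and physics alphas and a weaker force.
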AnimBlueprint /Game/Gadgets/Assets/Supply_Drop_Donut/Meshes/Supply_Drop_Donut_Gadget_AnimBP
target_skeleton=&/Supply_Drop_Donut_Gadget_Skeleton
graphs [
 AnimGraph {nodes [
   n1 = Get noise_alpha() -> (noise_alpha:double)
   n2 = Get noise_alpha()
   n3 = "Transform (Modify) Bone - Bone: offset"(
    ComponentPose=Local To Component(LocalPose=Input Pose())
    Alpha=@n2
    Translation="0,0,0"
    Rotation=Get noise_rotator() -> (noise_rotator:Rotator)
    Scale="1,1,1"
   ) {rotationMode=BMM_Additive, rotationSpace=BCS_BoneSpace, alphaBoolBlend {blendOption=Linear}, boneToModify {boneName=offset}}
   n4 = "Transform (Modify) Bone - Bone: pelvis"(ComponentPose=@n3, Alpha=@n2, Translation="0,0,0", Rotation=Get pelvis_noise_rotator() -> (pelvis_noise_rotator:Rotator), Scale="1,1,1") {rotationMode=BMM_Additive, rotationSpace=BCS_BoneSpace, alphaBoolBlend {blendOption=Linear}, boneToModify {boneName=pelvis}}
   n5 = RigidBody!(
    ComponentPose=@n4
    Alpha=Get physics_alpha() -> (physics_alpha:double)
    OverrideWorldGravity:Vector=Get gravity_override() -> (gravity_override:Vector)
    ExternalForce=Get vector_force() -> (vector_force:Vector)
   ) -> (Pose) {
    overridePhysicsAsset=&/Supply_Drop_Donut_Gadget_Physics
    alphaBoolBlend {blendOption=Linear}
    baseBoneRef {boneName=root}
    componentAppliedLinearAccClamp=<1000 1000 1000>
    componentLinearAccScale=<0.7 0.7 0.7>
    componentLinearVelScale=<0.7 0.7 0.7>
    overrideWorldGravity {z=-100}
   }
   n6 = Get noise_alpha()
   n7 = "Transform (Modify) Bone - Bone: clavicle_l"(ComponentPose=@n5, Alpha=@n6, Translation="0,0,0", Rotation=Get clavicle_noise_l() -> (clavicle_noise_l:Rotator), Scale="1,1,1") {rotationMode=BMM_Additive, rotationSpace=BCS_BoneSpace, alphaBoolBlend {blendOption=Linear}, boneToModify {boneName=clavicle_l}}
   n8 = "Transform (Modify) Bone - Bone: hand_l"(ComponentPose=@n7, Alpha=@n6, Translation="0,0,0", Rotation=Get hand_noise_l() -> (hand_noise_l:Rotator), Scale="1,1,1") {rotationMode=BMM_Additive, rotationSpace=BCS_BoneSpace, alphaBoolBlend {blendOption=Linear}, boneToModify {boneName=hand_l}}
   n9 = "Transform (Modify) Bone - Bone: thigh_l"(ComponentPose=@n8, Alpha=@n6, Translation="0,0,0", Rotation=Get thigh_noise_l() -> (thigh_noise_l:Rotator), Scale="1,1,1") {rotationMode=BMM_Additive, rotationSpace=BCS_BoneSpace, alphaBoolBlend {blendOption=Linear}, boneToModify {boneName=thigh_l}}
   n10 = "Transform (Modify) Bone - Bone: calf_l"(ComponentPose=@n9, Alpha=@n6, Translation="0,0,0", Rotation=Get calf_noise_l() -> (calf_noise_l:Rotator), Scale="1,1,1") {rotationMode=BMM_Additive, rotationSpace=BCS_BoneSpace, alphaBoolBlend {blendOption=Linear}, boneToModify {boneName=calf_l}}
   n11 = "Transform (Modify) Bone - Bone: clavicle_r"(ComponentPose=@n10, Alpha=@n1, Translation="0,0,0", Rotation=Get clavicle_noise_r() -> (clavicle_noise_r:Rotator), Scale="1,1,1") {rotationMode=BMM_Additive, rotationSpace=BCS_BoneSpace, alphaBoolBlend {blendOption=Linear}, boneToModify {boneName=clavicle_r}}
   n12 = "Transform (Modify) Bone - Bone: hand_r"(ComponentPose=@n11, Alpha=@n1, Translation="0,0,0", Rotation=Get hand_noise_r() -> (hand_noise_r:Rotator), Scale="1,1,1") {rotationMode=BMM_Additive, rotationSpace=BCS_BoneSpace, alphaBoolBlend {blendOption=Linear}, boneToModify {boneName=hand_r}}
   n13 = "Transform (Modify) Bone - Bone: thigh_r"(ComponentPose=@n12, Alpha=@n1, Translation="0,0,0", Rotation=Get thigh_noise_r() -> (thigh_noise_r:Rotator), Scale="1,1,1") {rotationMode=BMM_Additive, rotationSpace=BCS_BoneSpace, alphaBoolBlend {blendOption=Linear}, boneToModify {boneName=thigh_r}}
   n14 = "Transform (Modify) Bone - Bone: calf_r"(ComponentPose=@n13, Alpha=@n1, Translation="0,0,0", Rotation=Get calf_noise_r() -> (calf_noise_r:Rotator), Scale="1,1,1") {rotationMode=BMM_Additive, rotationSpace=BCS_BoneSpace, alphaBoolBlend {blendOption=Linear}, boneToModify {boneName=calf_r}}
   Output Pose(Result=Component To Local(ComponentPose=@n14)) {layerGroup=None}
  ]}
 EventGraph {nodes [
   OnLanded() {is=title K2Node_CustomEvent}
   > Set bLanded(bLanded:bool=true) -> (Output_Get:bool)
   > Set noise_alpha(noise_alpha:double=RandomFloatInRange(Min=0.185, Max=0.42)) -> (Output_Get:double)
   n1 = RandomFloatInRange(Min=50, Max=600)
   n2 = RandomFloatInRange(Min=100, Max=1000)
   > Set vector_force(vector_force:Vector="To Vector (Vector4)"!(InVector4:"Vector 4 Structure (by ref)"=MakeVector4!(X:double=@n2, Y:double=@n2, Z:double=@n1, W:double=0) -> (ReturnValue:Vector 4 Structure)) -> (ReturnValue:Vector)) -> (Output_Get:Vector)
   > Set physics_alpha(physics_alpha:double=RandomFloatInRange(Min=0.1, Max=0.65)) -> (Output_Get:double)
   Event BlueprintUpdateAnimation()
   > n3 = Sequence()
   n4 = Get noise_alpha()
   n5 = Get main_freq_rotator() -> (main_freq_rotator:Rotator)
   > Set noise_rotator(noise_rotator:Rotator=MakeWaveRotator!(
     Amplitudes:Rotator="2.5,2.5,2"
     Frequencies:Rotator="Scale Rotator (Float)"!(A:Rotator=@n5, B:float=@n4) -> (ReturnValue:Rotator)
     Offsets:Rotator=Get offset_rotator() -> (offset_rotator:Rotator)
     bUseNoise:bool=true
     Min:Vector
     Max:Vector
    ) -> (ReturnValue:Rotator)) -> (Output_Get:Rotator)
   n6 = Get noise_alpha()
   n7 = Get pelvis_freq_rotator() -> (pelvis_freq_rotator:Rotator)
   > Set pelvis_noise_rotator(pelvis_noise_rotator:Rotator=MakeWaveRotator(Amplitudes="2,1.5,2", Frequencies="Scale Rotator (Float)"(A=@n7, B=@n6), Offsets=Get offset_rotator(), bUseNoise=true)) -> (Output_Get:Rotator)
   n8 = Get noise_alpha()
   n9 = Get clav_freq_rot() -> (clav_freq_rot:Rotator)
   > Set clavicle_noise_l(clavicle_noise_l:Rotator=MakeWaveRotator(Amplitudes="5,6,8", Frequencies="Scale Rotator (Float)"(A=@n9, B=@n8), Offsets=Get offset_rotator(), bUseNoise=true)) -> (Output_Get:Rotator)
   n10 = Get noise_alpha()
   n11 = Get hand_freq_rot() -> (hand_freq_rot:Rotator)
   > Set hand_noise_l(hand_noise_l:Rotator=MakeWaveRotator(Amplitudes="8,3,5", Frequencies="Scale Rotator (Float)"(A=@n11, B=@n10), Offsets=Get offset_rotator(), bUseNoise=true)) -> (Output_Get:Rotator)
   n12 = Get noise_alpha()
   n13 = Get thigh_freq_rot() -> (thigh_freq_rot:Rotator)
   > Set thigh_noise_l(thigh_noise_l:Rotator=MakeWaveRotator(Amplitudes="2,6,1", Frequencies="Scale Rotator (Float)"(A=@n13, B=@n12), Offsets=Get offset_rotator(), bUseNoise=true)) -> (Output_Get:Rotator)
   n14 = Get calf_freq_rot() -> (calf_freq_rot:Rotator)
   n15 = Get noise_alpha()
   > Set calf_noise_l(calf_noise_l:Rotator=MakeWaveRotator(Amplitudes="4,7.35,1", Frequencies="Scale Rotator (Float)"(A=@n14, B=@n15), Offsets=Get offset_rotator(), bUseNoise=true)) -> (Output_Get:Rotator)
   n16 = Get clav_freq_r() -> (clav_freq_r:Rotator)
   n17 = Get noise_alpha()
   > Set clavicle_noise_r(clavicle_noise_r:Rotator=MakeWaveRotator(Amplitudes="6,8,3", Frequencies="Scale Rotator (Float)"(A=@n16, B=@n17), Offsets=Get offset_rotator(), bUseNoise=true)) -> (Output_Get:Rotator)
   n18 = Get hand_freq_r() -> (hand_freq_r:Rotator)
   n19 = Get noise_alpha()
   > Set hand_noise_r(hand_noise_r:Rotator=MakeWaveRotator(Amplitudes="3,3,5", Frequencies="Scale Rotator (Float)"(A=@n18, B=@n19), Offsets=Get offset_rotator(), bUseNoise=true)) -> (Output_Get:Rotator)
   n20 = Get thigh_freq_r() -> (thigh_freq_r:Rotator)
   n21 = Get noise_alpha()
   > Set thigh_noise_r(thigh_noise_r:Rotator=MakeWaveRotator(Amplitudes="3.5,6,1", Frequencies="Scale Rotator (Float)"(A=@n20, B=@n21), Offsets=Get offset_rotator(), bUseNoise=true)) -> (Output_Get:Rotator)
   n22 = Get calf_freq_r() -> (calf_freq_r:Rotator)
   n23 = Get noise_alpha()
   > Set calf_noise_r(calf_noise_r:Rotator=MakeWaveRotator(Amplitudes="4,7.35,1", Frequencies="Scale Rotator (Float)"(A=@n22, B=@n23), Offsets=Get offset_rotator(), bUseNoise=true)) -> (Output_Get:Rotator)
   n24 = Get random_offset() -> (random_offset:double)
   Set offset_rotator(
    execute=@n3.then_1
    offset_rotator:Rotator=MakeRotator(Roll=float + float(A=90, B=@n24) {is=title K2Node_CommutativeAssociativeBinaryOperator}, Yaw=float + float(A=45, B=@n24))
   ) -> (Output_Get:Rotator)
   Event BlueprintInitializeAnimation()
   > Set gravity_override(gravity_override="0, 0,-100")
   > Set vector_force(vector_force="50000,50000,10000")
   > Set random_offset(random_offset:double=RandomFloatInRange(Min=0, Max=180)) -> (Output_Get:double)
   TryGetPawnOwner()
  ]}
]
variables [
 bLanded {type=boolean}
 calf_freq_r {type=Rotator, default=<rot 120 99 100>}
 calf_freq_rot {type=Rotator, default=<rot 70 80 100>}
 calf_noise_l {type=Rotator}
 calf_noise_r {type=Rotator}
 clav_freq_r {type=Rotator, default=<rot 130 110 125>}
 clav_freq_rot {type=Rotator, default=<rot 90 85 100>}
 clavicle_noise_l {type=Rotator}
 clavicle_noise_r {type=Rotator}
 gravity_override {type=Vector}
 hand_freq_r {type=Rotator, default=<rot 125 180 150>}
 hand_freq_rot {type=Rotator, default=<rot 160 145 150>}
 hand_noise_l {type=Rotator}
 hand_noise_r {type=Rotator}
 main_freq_rotator {type=Rotator, default=<rot 180 50 200>}
 noise_alpha {type=number, default=0.8}
 noise_rotator {type=Rotator}
 offset_rotator {type=Rotator}
 pelvis_freq_rotator {type=Rotator, default=<rot 400 50 300>}
 pelvis_noise_rotator {type=Rotator}
 physics_alpha {type=number, default=1}
 random_offset {type=number, default=0.4}
 thigh_freq_r {type=Rotator, default=<rot 130 130 80>}
 thigh_freq_rot {type=Rotator, default=<rot 100 100 100>}
 thigh_noise_l {type=Rotator}
 thigh_noise_r {type=Rotator}
 vector_force {type=Vector}
]
